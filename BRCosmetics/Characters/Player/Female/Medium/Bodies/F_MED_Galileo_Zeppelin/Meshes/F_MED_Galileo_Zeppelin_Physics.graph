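# Physics asset for the Galileo Zeppelin medium female player body cosmetic, covering only its dynamic belt bones.
# Capsule bodies on the left, right and back belt bones each constrain directly to a tiny capsule on dyn_parent_belt, forming a star rather than a chain.
# Tight twist and angular limits let the side pieces swing mostly on one axis, while the back piece also has moderate play on the other swing axis.
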PhysicsAsset /BRCosmetics/Characters/Player/Female/Medium/Bodies/F_MED_Galileo_Zeppelin/Meshes/F_MED_Galileo_Zeppelin_Physics
bodies {
 dyn_belt_01_l [{shapeType=Capsule, length=5.75, radius=7, center=<1.5 0 0>, rotation=<rot -90 0 0>}]
 dyn_belt_01_r [{shapeType=Capsule, length=11.25, radius=6.25, center=<1.5 0 0>, rotation=<rot -90 0 0>}]
 dyn_belt_bck_01 [{shapeType=Capsule, length=6, radius=2.5, center=<0.15 0 0>, rotation=<rot -90 0 0>}]
 dyn_parent_belt [{shapeType=Capsule, length=0.505, radius=0.505, rotation=<rot 0 -90 0>}]
}
constraints [
 {bone1Name=dyn_belt_01_r, bone2Name=dyn_parent_belt, swing1LimitDegrees=1, swing2LimitDegrees=20, twistLimitDegrees=1}
 {bone1Name=dyn_belt_01_l, bone2Name=dyn_parent_belt, swing1LimitDegrees=1, swing2LimitDegrees=15, twistLimitDegrees=1}
 {bone1Name=dyn_belt_bck_01, bone2Name=dyn_parent_belt, swing1LimitDegrees=9.5714, swing2LimitDegrees=20, twistLimitDegrees=1}
]
