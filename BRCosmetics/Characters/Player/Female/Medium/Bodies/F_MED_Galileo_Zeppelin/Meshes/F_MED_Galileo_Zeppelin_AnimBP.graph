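# Post-process-style cosmetic AnimBlueprint for the F_MED_Galileo_Zeppelin body, copying its pose from the attached parent mesh.
# Constraints place gun-parent bones, then alpha-driven additive bone transforms correct glutes, knees and gun drivers using thigh and calf variables not declared here.
# A RigidBody node simulates with its own physics asset by BeltAlpha, before weighted backpack attachment and a front-end idle tweak.
# Belt AnimDynamics nodes exist but are disconnected from the output.
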AnimBlueprint /BRCosmetics/Characters/Player/Female/Medium/Bodies/F_MED_Galileo_Zeppelin/Meshes/F_MED_Galileo_Zeppelin_AnimBP
target_skeleton=&/F_MED_Galileo_Zeppelin_Skeleton
graphs [
 AnimGraph {nodes [
   Anim Dynamics - Bone: dyn_belt_01_r(Alpha=0.8, ExternalForce="0,0,0") {
    angularDampingOverride=1
    angularSpringConstant=1500
    bAngularSpring=true
    bGravityOverrideInSimSpace=true
    bOverrideAngularDamping=true
    bUseGravityOverride=true
    gravityScale=0.35
    simulationSpace=RootRelative
    alphaBoolBlend {blendOption=Linear}
    alphaScaleBiasClamp {bInterpResult=true}
    boundBone {boneName=dyn_belt_01_r}
    componentAppliedLinearAccClamp=<3000 3000 3000>
    componentLinearAccScale=<0.3 0.3 0.3>
    componentLinearVelScale=<0.3 0.3 1>
    gravityOverride {z=-980}
    physicsBodyDefinitions [{
      boundBone {boneName=dyn_belt_01_r}
      boxExtents=<20 20 10>
      constraintSetup {angularLimitsMax=<0 0 0>, angularLimitsMin=<0 -25 0>, angularTarget=<0 0 0>}
      localJointOffset=<0 0 -8>
     }]
   }
   n1 = Anim Dynamics - Bone: dyn_belt_frnt_01(Alpha=1, ExternalForce="0,0,0") {
    angularDampingOverride=1
    angularSpringConstant=1500
    bAngularSpring=true
    bGravityOverrideInSimSpace=true
    bOverrideAngularDamping=true
    bUseGravityOverride=true
    gravityScale=0.35
    simulationSpace=RootRelative
    alphaBoolBlend {blendOption=Linear}
    boundBone {boneName=dyn_belt_frnt_01}
    componentAppliedLinearAccClamp=<3000 3000 3500>
    componentLinearAccScale=<0.3 0.3 0.3>
    componentLinearVelScale=<0.3 0.3 1.5>
    gravityOverride {z=-980}
    physicsBodyDefinitions [{
      sphereCollisionRadius=0
      boundBone {boneName=dyn_belt_frnt_01}
      boxExtents=<20 20 10>
      constraintSetup {angularLimitsMax=<5 0 5>, angularLimitsMin=<-5 -45 -5>, angularTarget=<0 0 0>}
      localJointOffset=<4 0 0>
     }]
   }
   Anim Dynamics - Bone: dyn_belt_bck_01(ComponentPose=@n1, Alpha=1, ExternalForce="0,0,0") {
    angularDampingOverride=1
    angularSpringConstant=1500
    bAngularSpring=true
    bGravityOverrideInSimSpace=true
    bOverrideAngularDamping=true
    bUseGravityOverride=true
    gravityScale=0.35
    simulationSpace=RootRelative
    alphaBoolBlend {blendOption=Linear}
    boundBone {boneName=dyn_belt_bck_01}
    componentAppliedLinearAccClamp=<3000 3000 3000>
    componentLinearAccScale=<0.3 0.3 0.3>
    componentLinearVelScale=<0.3 0.3 2>
    gravityOverride {z=-980}
    physicsBodyDefinitions [{
      sphereCollisionRadius=0
      boundBone {boneName=dyn_belt_bck_01}
      boxExtents=<15 20 10>
      constraintSetup {angularLimitsMax=<1 0 5>, angularLimitsMin=<-1 -45 -3>, angularTarget=<0 0 0>}
      localJointOffset=<4 0 0>
     }]
   }
   n2 = Anim Dynamics - Bone: dyn_belt_01_l(Alpha=0.7, ExternalForce="0,0,0") {
    angularDampingOverride=1
    angularSpringConstant=1500
    bAngularSpring=true
    bGravityOverrideInSimSpace=true
    bOverrideAngularDamping=true
    bOverrideLinearDamping=true
    bUseGravityOverride=true
    gravityScale=0.35
    linearDampingOverride=0.9999
    alphaBoolBlend {blendOption=Linear}
    boundBone {boneName=dyn_belt_01_l}
    componentAppliedLinearAccClamp=<100 100 1500>
    componentLinearAccScale=<0.3 0.3 0.3>
    componentLinearVelScale=<0.3 0.3 0.3>
    gravityOverride {z=-980}
    physicsBodyDefinitions [{
      sphereCollisionRadius=0
      boundBone {boneName=dyn_belt_01_l}
      boxExtents=<15 20 10>
      constraintSetup {angularLimitsMax=<2 3 2>, angularLimitsMin=<-2 -12 -2>, angularTarget=<1 0 0>}
      localJointOffset=<0 0 -1>
     }]
   }
   Anim Dynamics - Bone: dyn_belt_01_l(ComponentPose=@n2, Alpha=0.7, ExternalForce="0,0,0") {
    angularDampingOverride=1
    angularSpringConstant=1500
    bAngularSpring=true
    bGravityOverrideInSimSpace=true
    bOverrideAngularDamping=true
    bOverrideLinearDamping=true
    bUseGravityOverride=true
    gravityScale=0.35
    linearDampingOverride=0.9999
    alphaBoolBlend {blendOption=Linear}
    boundBone {boneName=dyn_belt_01_l}
    componentAppliedLinearAccClamp=<100 100 1500>
    componentLinearAccScale=<0.3 0.3 0.3>
    componentLinearVelScale=<0.3 0.3 0.3>
    gravityOverride {z=-980}
    physicsBodyDefinitions [{
      sphereCollisionRadius=0
      boundBone {boneName=dyn_belt_01_l}
      boxExtents=<15 20 10>
      constraintSetup {angularLimitsMax=<2 3 2>, angularLimitsMin=<-2 -12 -2>, angularTarget=<1 0 0>}
      localJointOffset=<0 0 -1>
     }]
   }
   n3 = "Constraint - dyn_parent_gun_l (PP)"!(
    ComponentPose=Local To Component(LocalPose=Copy Pose From Mesh() {bUseAttachedParent=true})
    Alpha=1
    ConstraintWeights_0=0.2388
    ConstraintWeights_1:float=0.7612
   ) -> (Pose) {
    alphaBoolBlend {blendOption=Linear}
    boneToModify {boneName=dyn_parent_gun_l}
    constraintSetup [{targetBone {boneName=thigh_twist_01_l}}, {targetBone {boneName=thigh_l}}]
    constraintWeights [1,1]
   }
   n4 = "Constraint - dyn_parent_gun_r (PP)"!(ComponentPose=@n3, Alpha=1, ConstraintWeights_0=0.7596, ConstraintWeights_1:float=0.2404) -> (Pose) {
    alphaBoolBlend {blendOption=Linear}
    boneToModify {boneName=dyn_parent_gun_r}
    constraintSetup [{targetBone {boneName=thigh_r}}, {targetBone {boneName=thigh_twist_01_r}}]
    constraintWeights [1,1]
   }
   n5 = "Transform (Modify) Bone - Bone: dyn_glute_l"(
    ComponentPose=@n4
    Alpha=Get ThighLeftAlpha() -> (ThighLeftAlpha:float)
    Translation="-3,-8,-5"
    Rotation="0.000,0,-58"
    Scale="1.35,1.35,1.35"
   ) {
    rotationMode=BMM_Additive
    rotationSpace=BCS_BoneSpace
    scaleMode=BMM_Additive
    scaleSpace=BCS_BoneSpace
    translationMode=BMM_Additive
    translationSpace=BCS_BoneSpace
    alphaBoolBlend {blendOption=Linear}
    boneToModify {boneName=dyn_glute_l}
    rotation {roll=-58}
    scale=<1.35 1.35 1.35>
    translation=<-3 -8 -5>
   }
   n6 = "Transform (Modify) Bone - Bone: dyn_glute_r"(
    ComponentPose=@n5
    Alpha=Get ThighRightAlpha() -> (ThighRightAlpha:float)
    Translation="3,-8,-5"
    Rotation="0.000,0,-58"
    Scale="1.35,1.35,1.35"
   ) {
    rotationMode=BMM_Additive
    rotationSpace=BCS_BoneSpace
    scaleMode=BMM_Additive
    scaleSpace=BCS_BoneSpace
    translationMode=BMM_Additive
    translationSpace=BCS_BoneSpace
    alphaBoolBlend {blendOption=Linear}
    boneToModify {boneName=dyn_glute_r}
    rotation {roll=-58}
    scale=<1.35 1.35 1.35>
    translation=<3 -8 -5>
   }
   n7 = "Transform (Modify) Bone - Bone: dyn_knee_r"(ComponentPose=@n6, Alpha=Get CalfRightAlpha() -> (CalfRightAlpha:float), Translation="-5,0,0", Rotation="0.000,20,0.000", Scale="1,1,1") {
    rotationMode=BMM_Additive
    rotationSpace=BCS_BoneSpace
    scaleMode=BMM_Additive
    scaleSpace=BCS_BoneSpace
    translationMode=BMM_Additive
    translationSpace=BCS_BoneSpace
    alphaBoolBlend {blendOption=Linear}
    boneToModify {boneName=dyn_knee_r}
    rotation {yaw=20}
    translation {x=-5}
   }
   n8 = "Transform (Modify) Bone - Bone: dyn_knee_l"(ComponentPose=@n7, Alpha=Get CalfLeftAlpha() -> (CalfLeftAlpha:float), Translation="5,0,0", Rotation="0.000,20,0.000", Scale="1,1,1") {
    rotationMode=BMM_Additive
    rotationSpace=BCS_BoneSpace
    scaleMode=BMM_Additive
    scaleSpace=BCS_BoneSpace
    translationMode=BMM_Additive
    translationSpace=BCS_BoneSpace
    alphaBoolBlend {blendOption=Linear}
    boneToModify {boneName=dyn_knee_l}
    rotation {yaw=20}
    translation {x=5}
   }
   n9 = "Transform (Modify) Bone - Bone: dyn_driver_gun_l"(ComponentPose=@n8, Alpha=Get ThighLeftAlpha(), Translation="0,0,0", Rotation="0.000,30,0.000", Scale="1,1,1") {
    rotationMode=BMM_Additive
    rotationSpace=BCS_BoneSpace
    scaleSpace=BCS_BoneSpace
    translationSpace=BCS_BoneSpace
    alphaBoolBlend {blendOption=Linear}
    alphaScaleBiasClamp {bInterpResult=true, interpSpeedDecreasing=20, interpSpeedIncreasing=20}
    boneToModify {boneName=dyn_driver_gun_l}
    rotation {yaw=30}
   }
   n10 = "Transform (Modify) Bone - Bone: dyn_driver_gun_r"(ComponentPose=@n9, Alpha=Get ThighRightAlpha(), Translation="0,0,0", Rotation="0,-30,0", Scale="1,1,1") {
    rotationMode=BMM_Additive
    rotationSpace=BCS_BoneSpace
    scaleMode=BMM_Additive
    scaleSpace=BCS_BoneSpace
    translationMode=BMM_Additive
    translationSpace=BCS_BoneSpace
    alphaBoolBlend {blendOption=Linear}
    alphaScaleBiasClamp {bInterpResult=true, interpSpeedDecreasing=20, interpSpeedIncreasing=20}
    boneToModify {boneName=dyn_driver_gun_r}
    rotation {yaw=-30}
   }
   n11 = RigidBody(ComponentPose=@n10, Alpha=Get BeltAlpha() -> (BeltAlpha:float)) {
    bClampLinearTranslationLimitToRefPose=true
    bOverrideWorldGravity=true
    overridePhysicsAsset=&/F_MED_Galileo_Zeppelin_Physics
    simulationSpace=BaseBoneSpace
    alphaBoolBlend {blendOption=Linear}
    baseBoneRef {boneName=root}
    componentAppliedLinearAccClamp=<1100 1100 1100>
    componentLinearAccScale=<0.7 0.7 0.25>
    componentLinearVelScale=<0.7 0.7 0.25>
    overrideWorldGravity {z=-980}
   }
   n12 = "Constraint - attach_backpack (PPPPPP)"!(
    ComponentPose=@n11
    Alpha=1
    ConstraintWeights_0=0.1606
    ConstraintWeights_1:float=0.0011
    ConstraintWeights_2:float=0.0738
    ConstraintWeights_3:float=0.5844
    ConstraintWeights_4:float=0.0001
    ConstraintWeights_5:float=0.18
   ) -> (Pose) {
    alphaBoolBlend {blendOption=Linear}
    boneToModify {boneName=attach_backpack}
    constraintSetup [
     {targetBone {boneName=clavicle_l}}
     {targetBone {boneName=upperarm_twist_01_r}}
     {targetBone {boneName=spine_04}}
     {targetBone {boneName=spine_05}}
     {targetBone {boneName=upperarm_twist_01_l}}
     {targetBone {boneName=clavicle_r}}
    ]
    constraintWeights [1,1,1,1,1,1]
   }
   n13 = "Transform (Modify) Bone - Bone: attach_backpack"(
    ComponentPose=@n12
    Alpha=Get FrontEndIdleAlpha() -> (FrontEndIdleAlpha:float)
    Translation="0,0,0"
    Rotation="0.000,-7,0.000"
    Scale="1,1,1"
   ) {
    rotationMode=BMM_Additive
    rotationSpace=BCS_BoneSpace
    alphaBoolBlend {blendOption=Linear}
    boneToModify {boneName=attach_backpack}
    rotation {yaw=-7}
   }
   Output Pose(Result=Component To Local(ComponentPose=@n13)) {layerGroup=None}
  ]}
 EventGraph {links [], nodes []}
]
variables []
